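# Sound cue for Labrador body fall foley, reusing Burt creature body fall recordings at doubled pitch and slightly reduced volume.
# A quality switch picks an equally weighted random from four variations for the first two levels, or from just the first two variations for the other levels.
# It routes to the creature foley sound class, with prey gallop attenuation and owner-limited and prey group concurrency.
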SoundCue /Labrador/Sound/Foley/Labrador_Foley_BodyFall_Cue
root=QualityLevel_0
nodes [
 {id=QualityLevel_0, children [Random_1,Random_1,Random_3,Random_3]}
 {id=Random_1, children [WavePlayer_1,WavePlayer_2,WavePlayer_4,WavePlayer_5], properties {weights [1,1,1,1]}}
 {id=Random_3, children [WavePlayer_6,WavePlayer_7], properties {weights [1,1]}}
 {id=WavePlayer_1, properties {soundWaveAssetPtr=/Irwin/Sounds/Burt/Foley/Burt_Body_BodyFall_01}}
 {id=WavePlayer_2, properties {soundWaveAssetPtr=/Irwin/Sounds/Burt/Foley/Burt_Body_BodyFall_02}}
 {id=WavePlayer_4, properties {soundWaveAssetPtr=/Irwin/Sounds/Burt/Foley/Burt_Body_BodyFall_03}}
 {id=WavePlayer_5, properties {soundWaveAssetPtr=/Irwin/Sounds/Burt/Foley/Burt_Body_BodyFall_04}}
 {id=WavePlayer_6, properties {soundWaveAssetPtr=&Burt_Body_BodyFall_01}}
 {id=WavePlayer_7, properties {soundWaveAssetPtr=&Burt_Body_BodyFall_02}}
]
properties {
 attenuationSettings=/Irwin/Sounds/Settings/Prey_Gallop_Att
 maxDistance=3200
 pitchMultiplier=2
 soundClassObject=/Game/Sounds/FortSoundClassesAndModes/Fort_Creatures_Foley
 volumeMultiplier=0.7
 concurrencySet [/Irwin/Sounds/Settings/Conc/SetOneLimitToOwner_Foley_Conc, /Irwin/Sounds/Settings/Conc/Prey_SetGroup_Foley_Conc]
}
