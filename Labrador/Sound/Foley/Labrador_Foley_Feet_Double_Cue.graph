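# Sound cue for Labrador creature double-footstep foley, using prey step attenuation and the creature steps sound class.
# A quality level node sends its first two slots to an equal-weight random pick of six wave variants and its last two to a pick of just the first two variants.
# A modulator at the root varies pitch and volume per play.
# Two foley concurrency sets, one per-owner and one for the prey group, limit overlapping plays.
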SoundCue /Labrador/Sound/Foley/Labrador_Foley_Feet_Double_Cue
root=Modulator_1
nodes [
 {id=Modulator_1, children [QualityLevel_1], properties {pitchMax=1.2, pitchMin=1, volumeMax=0.9, volumeMin=0.75}}
 {id=QualityLevel_1, children [Random_2,Random_2,Random_3,Random_3]}
 {
  id=Random_2
  children [WavePlayer_7, WavePlayer_8, WavePlayer_9, WavePlayer_10, WavePlayer_11, WavePlayer_12]
  properties {weights [1,1,1,1,1,1]}
 }
 {id=Random_3, children [WavePlayer_13, WavePlayer_14], properties {weights [1,1]}}
 {id=WavePlayer_10, properties {soundWaveAssetPtr=&/Labrador_Foley_Feet_Double_04}}
 {id=WavePlayer_11, properties {soundWaveAssetPtr=&/Labrador_Foley_Feet_Double_05}}
 {id=WavePlayer_12, properties {soundWaveAssetPtr=&/Labrador_Foley_Feet_Double_06}}
 {id=WavePlayer_13, properties {soundWaveAssetPtr=&/Labrador_Foley_Feet_Double_01}}
 {id=WavePlayer_14, properties {soundWaveAssetPtr=&/Labrador_Foley_Feet_Double_02}}
 {id=WavePlayer_7, properties {soundWaveAssetPtr=&Labrador_Foley_Feet_Double_01}}
 {id=WavePlayer_8, properties {soundWaveAssetPtr=&Labrador_Foley_Feet_Double_02}}
 {id=WavePlayer_9, properties {soundWaveAssetPtr=&/Labrador_Foley_Feet_Double_03}}
]
properties {
 attenuationSettings=/Irwin/Sounds/Settings/Prey_Step_Att
 maxDistance=3200
 pitchMultiplier=1.1
 soundClassObject=/Game/Sounds/FortSoundClassesAndModes/Fort_Creatures_Steps_Normal
 volumeMultiplier=1
 concurrencySet [/Irwin/Sounds/Settings/Conc/SetOneLimitToOwner_Foley_Conc, /Irwin/Sounds/Settings/Conc/Prey_SetGroup_Foley_Conc]
}
