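# Sound cue for Labrador galloping footstep foley, using creature-step sound class, far gallop prey attenuation and concurrency limits.
# A mixer blends a pitch- and volume-varied footfall with a quieter, briefly delayed, higher-pitched "boing" accent picked randomly from six waves.
# A quality-level node feeds the footfall layer, using nine random gallop variants for the first two levels and a four-variant subset for the last two.
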SoundCue /Labrador/Sound/Foley/Labrador_Foley_Feet_Gallop_Cue
root=Mixer_0
nodes [
 {id=Delay_1, children [Modulator_0], properties {delayMax=0.1, delayMin=0.1}}
 {id=Mixer_0, children [Modulator_1,Delay_1], properties {inputVolume [1,0.55]}}
 {id=Modulator_0, children [Random_0], properties {pitchMax=1.55, pitchMin=1.4}}
 {id=Modulator_1, children [QualityLevel_1], properties {pitchMax=1.2, pitchMin=1, volumeMax=1, volumeMin=0.8}}
 {id=QualityLevel_1, children [Random_1,Random_1,Random_3,Random_3]}
 {
  id=Random_0
  children [WavePlayer_0, WavePlayer_1, WavePlayer_2, WavePlayer_3, WavePlayer_16, WavePlayer_17]
  properties {weights [1,1,1,1,1,1]}
 }
 {
  id=Random_1
  children [WavePlayer_4, WavePlayer_5, WavePlayer_6, WavePlayer_7, WavePlayer_8, WavePlayer_9, WavePlayer_10, WavePlayer_11, WavePlayer_12]
  properties {weights [1,1,1,1,1,1,1,1,1]}
 }
 {id=Random_3, children [WavePlayer_13, WavePlayer_14, WavePlayer_15, WavePlayer_18], properties {weights [1,1,1,1]}}
 {id=WavePlayer_0, properties {soundWaveAssetPtr=&/Labrador_Gallop_Boing_06}}
 {id=WavePlayer_1, properties {soundWaveAssetPtr=&/Labrador_Gallop_Boing_05}}
 {id=WavePlayer_10, properties {soundWaveAssetPtr=&/Labrador_Foley_Feet_Gallop_07}}
 {id=WavePlayer_11, properties {soundWaveAssetPtr=&/Labrador_Foley_Feet_Gallop_08}}
 {id=WavePlayer_12, properties {soundWaveAssetPtr=&/Labrador_Foley_Feet_Gallop_09}}
 {id=WavePlayer_13, properties {soundWaveAssetPtr=&/Labrador_Foley_Feet_Gallop_01}}
 {id=WavePlayer_14, properties {soundWaveAssetPtr=&/Labrador_Foley_Feet_Gallop_02}}
 {id=WavePlayer_15, properties {soundWaveAssetPtr=&/Labrador_Foley_Feet_Gallop_03}}
 {id=WavePlayer_16, properties {soundWaveAssetPtr=&/Labrador_Gallop_Boing_02}}
 {id=WavePlayer_17, properties {soundWaveAssetPtr=&/Labrador_Gallop_Boing_01}}
 {id=WavePlayer_18, properties {soundWaveAssetPtr=&/Labrador_Foley_Feet_Gallop_04}}
 {id=WavePlayer_2, properties {soundWaveAssetPtr=&/Labrador_Gallop_Boing_04}}
 {id=WavePlayer_3, properties {soundWaveAssetPtr=&/Labrador_Gallop_Boing_03}}
 {id=WavePlayer_4, properties {soundWaveAssetPtr=&Labrador_Foley_Feet_Gallop_01}}
 {id=WavePlayer_5, properties {soundWaveAssetPtr=&Labrador_Foley_Feet_Gallop_02}}
 {id=WavePlayer_6, properties {soundWaveAssetPtr=&Labrador_Foley_Feet_Gallop_03}}
 {id=WavePlayer_7, properties {soundWaveAssetPtr=&Labrador_Foley_Feet_Gallop_04}}
 {id=WavePlayer_8, properties {soundWaveAssetPtr=&/Labrador_Foley_Feet_Gallop_05}}
 {id=WavePlayer_9, properties {soundWaveAssetPtr=&/Labrador_Foley_Feet_Gallop_06}}
]
properties {
 attenuationSettings=/Irwin/Sounds/Settings/Prey_Gallop_Far_Att
 maxDistance=4462.9136
 pitchMultiplier=1.1
 soundClassObject=/Game/Sounds/FortSoundClassesAndModes/Fort_Creatures_Steps_Normal
 volumeMultiplier=1
 concurrencySet [/Irwin/Sounds/Settings/Conc/SetOneLimitToOwner_Foley_Conc, /Irwin/Sounds/Settings/Conc/Prey_SetGroup_Foley_Conc]
}
