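# Sound cue for the Labrador creature's walking footsteps, mixing a step sound with a randomized squeak layer.
# A quality-level node picks from ten walk variations for the first two levels and a four-sample subset for the other two.
# The squeak layer uses RiftAnomaly double-squeaks behind a default-valued Delay node, pitched higher and quieter than the steps.
# Both layers randomize pitch and volume, using shared Irwin prey step attenuation and foley concurrency.
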SoundCue /Labrador/Sound/Foley/Labrador_Foley_Feet_Walk_Cue
root=Mixer_1
nodes [
 {id=Delay_1, children [Modulator_0]}
 {id=Mixer_1, children [Modulator_1,Delay_1], properties {inputVolume [1,1]}}
 {id=Modulator_0, children [Random_1], properties {pitchMax=1.5, pitchMin=1.1, volumeMax=0.8, volumeMin=0.6}}
 {id=Modulator_1, children [QualityLevel_1], properties {pitchMax=1.2, pitchMin=1, volumeMax=0.9, volumeMin=0.75}}
 {id=QualityLevel_1, children [Random_0,Random_0,Random_3,Random_3]}
 {
  id=Random_0
  children [
   WavePlayer_5
   WavePlayer_6
   WavePlayer_7
   WavePlayer_8
   WavePlayer_19
   WavePlayer_20
   WavePlayer_21
   WavePlayer_22
   WavePlayer_23
   WavePlayer_24
  ]
  properties {weights [1,1,1,1,1,1,1,1,1,1]}
 }
 {
  id=Random_1
  children [WavePlayer_4, WavePlayer_0, WavePlayer_1, WavePlayer_15, WavePlayer_18, WavePlayer_2]
  properties {weights [1,1,1,1,1,1]}
 }
 {id=Random_3, children [WavePlayer_25, WavePlayer_26, WavePlayer_27, WavePlayer_28], properties {weights [1,1,1,1]}}
 {id=WavePlayer_0, properties {soundWaveAssetPtr=&/Labrador_RiftAnomaly_Feet_Double_Squeak_02}}
 {id=WavePlayer_1, properties {soundWaveAssetPtr=&/Labrador_RiftAnomaly_Feet_Double_Squeak_03}}
 {id=WavePlayer_15, properties {soundWaveAssetPtr=&/Labrador_RiftAnomaly_Feet_Double_Squeak_04}}
 {id=WavePlayer_18, properties {soundWaveAssetPtr=&/Labrador_RiftAnomaly_Feet_Double_Squeak_05}}
 {id=WavePlayer_19, properties {soundWaveAssetPtr=&/Labrador_Foley_Feet_Walk_04}}
 {id=WavePlayer_2, properties {soundWaveAssetPtr=&/Labrador_RiftAnomaly_Feet_Double_Squeak_06}}
 {id=WavePlayer_20, properties {soundWaveAssetPtr=&/Labrador_Foley_Feet_Walk_05}}
 {id=WavePlayer_21, properties {soundWaveAssetPtr=&/Labrador_Foley_Feet_Walk_06}}
 {id=WavePlayer_22, properties {soundWaveAssetPtr=&/Labrador_Foley_Feet_Walk_07}}
 {id=WavePlayer_23, properties {soundWaveAssetPtr=&/Labrador_Foley_Feet_Walk_08}}
 {id=WavePlayer_24, properties {soundWaveAssetPtr=&/Labrador_Foley_Feet_Walk_09}}
 {id=WavePlayer_25, properties {soundWaveAssetPtr=&/Labrador_Foley_Feet_Walk_10}}
 {id=WavePlayer_26, properties {soundWaveAssetPtr=&/Labrador_Foley_Feet_Walk_01}}
 {id=WavePlayer_27, properties {soundWaveAssetPtr=&/Labrador_Foley_Feet_Walk_02}}
 {id=WavePlayer_28, properties {soundWaveAssetPtr=&/Labrador_Foley_Feet_Walk_03}}
 {id=WavePlayer_4, properties {soundWaveAssetPtr=&/Labrador_RiftAnomaly_Feet_Double_Squeak_01}}
 {id=WavePlayer_5, properties {soundWaveAssetPtr=&Labrador_Foley_Feet_Walk_10}}
 {id=WavePlayer_6, properties {soundWaveAssetPtr=&Labrador_Foley_Feet_Walk_01}}
 {id=WavePlayer_7, properties {soundWaveAssetPtr=&Labrador_Foley_Feet_Walk_02}}
 {id=WavePlayer_8, properties {soundWaveAssetPtr=&Labrador_Foley_Feet_Walk_03}}
]
properties {
 attenuationSettings=/Irwin/Sounds/Settings/Prey_Step_Att
 maxDistance=3200
 soundClassObject=/Game/Sounds/FortSoundClassesAndModes/Fort_Creatures_Steps_Normal
 volumeMultiplier=0.6
 concurrencySet [/Irwin/Sounds/Settings/Conc/Prey_SetGroup_Foley_Conc]
}
