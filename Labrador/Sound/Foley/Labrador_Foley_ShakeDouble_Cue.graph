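# Sound cue for the Labrador creature's double-shake foley.
# It picks at random between two equally weighted Burt ShakeDouble foley waves borrowed from the Irwin content, then slightly varies volume through a modulator.
# The cue doubles the pitch and lowers the volume, routing to the creature foley sound class with the shared Prey gallop attenuation and foley concurrency group.
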SoundCue /Labrador/Sound/Foley/Labrador_Foley_ShakeDouble_Cue
root=Modulator_0
nodes [
 {id=Modulator_0, children [Random_0], properties {volumeMax=1.2, volumeMin=1.1}}
 {id=Random_0, children [WavePlayer_0,WavePlayer_2], properties {bShouldExcludeFromBranchCulling=true, weights [1,1]}}
 {id=WavePlayer_0, properties {soundWaveAssetPtr=/Irwin/Sounds/Burt/Foley/Burt_Foley_ShakeDouble_01}}
 {id=WavePlayer_2, properties {soundWaveAssetPtr=/Irwin/Sounds/Burt/Foley/Burt_Foley_ShakeDouble_02}}
]
properties {
 attenuationSettings=/Irwin/Sounds/Settings/Prey_Gallop_Att
 maxDistance=3200
 pitchMultiplier=2
 soundClassObject=/Game/Sounds/FortSoundClassesAndModes/Fort_Creatures_Foley
 volumeMultiplier=0.2
 concurrencyOverrides {maxCount=3}
 concurrencySet [/Irwin/Sounds/Settings/Conc/Prey_SetGroup_Foley_Conc]
}
