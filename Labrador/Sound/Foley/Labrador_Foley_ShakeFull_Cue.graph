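# Sound cue for the Labrador creature's full-body shake foley, reusing Burt foley recordings from the Irwin plugin.
# A Random node picks one of three equally weighted shake takes, excluded from branch culling, and feeds a Modulator that varies volume slightly.
# The cue plays at doubled pitch and low volume through the creature foley sound class, with prey gallop attenuation and prey foley concurrency.
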SoundCue /Labrador/Sound/Foley/Labrador_Foley_ShakeFull_Cue
root=Modulator_0
nodes [
 {id=Modulator_0, children [Random_1], properties {volumeMax=1.2, volumeMin=1.1}}
 {id=Random_1, children [WavePlayer_3,WavePlayer_4,WavePlayer_5], properties {bShouldExcludeFromBranchCulling=true, weights [1,1,1]}}
 {id=WavePlayer_3, properties {soundWaveAssetPtr=/Irwin/Sounds/Burt/Foley/Burt_Foley_ShakeFull_01}}
 {id=WavePlayer_4, properties {soundWaveAssetPtr=/Irwin/Sounds/Burt/Foley/Burt_Foley_ShakeFull_03}}
 {id=WavePlayer_5, properties {soundWaveAssetPtr=/Irwin/Sounds/Burt/Foley/Burt_Foley_ShakeFull_02}}
]
properties {
 attenuationSettings=/Irwin/Sounds/Settings/Prey_Gallop_Att
 maxDistance=3200
 pitchMultiplier=2
 soundClassObject=/Game/Sounds/FortSoundClassesAndModes/Fort_Creatures_Foley
 volumeMultiplier=0.2
 concurrencySet [/Irwin/Sounds/Settings/Conc/Prey_SetGroup_Foley_Conc]
}
