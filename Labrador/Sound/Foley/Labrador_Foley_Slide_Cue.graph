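# Sound cue for Labrador creature slide foley, reusing Burt slide recordings from the Irwin content.
# A quality-level switch feeds the first two levels from a random pick of five waves and the last two from a three-wave subset.
# A modulator at the root varies pitch and volume.
# Prey step attenuation, a creature-steps sound class and two foley concurrency settings control range, mixing and voice limits.
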SoundCue /Labrador/Sound/Foley/Labrador_Foley_Slide_Cue
root=Modulator_0
nodes [
 {id=Modulator_0, children [QualityLevel_0], properties {pitchMax=1.1, pitchMin=0.9, volumeMax=0.9, volumeMin=0.75}}
 {id=QualityLevel_0, children [Random_1,Random_1,Random_0,Random_0]}
 {id=Random_0, children [WavePlayer_1, WavePlayer_2, WavePlayer_10], properties {bShouldExcludeFromBranchCulling=true, weights [1,1,1]}}
 {id=Random_1, children [WavePlayer_0,WavePlayer_6,WavePlayer_7,WavePlayer_8,WavePlayer_9], properties {weights [1,1,1,1,1]}}
 {id=WavePlayer_0, properties {soundWaveAssetPtr=/Irwin/Sounds/Burt/Foley/Burt_Slide_01}}
 {id=WavePlayer_1, properties {soundWaveAssetPtr=&Burt_Slide_01}}
 {id=WavePlayer_10, properties {soundWaveAssetPtr=/Irwin/Sounds/Burt/Foley/Burt_Slide_05}}
 {id=WavePlayer_2, properties {soundWaveAssetPtr=/Irwin/Sounds/Burt/Foley/Burt_Slide_02}}
 {id=WavePlayer_6, properties {soundWaveAssetPtr=&Burt_Slide_02}}
 {id=WavePlayer_7, properties {soundWaveAssetPtr=/Irwin/Sounds/Burt/Foley/Burt_Slide_03}}
 {id=WavePlayer_8, properties {soundWaveAssetPtr=/Irwin/Sounds/Burt/Foley/Burt_Slide_04}}
 {id=WavePlayer_9, properties {soundWaveAssetPtr=&Burt_Slide_05}}
]
properties {
 attenuationSettings=/Irwin/Sounds/Settings/Prey_Step_Att
 maxDistance=3200
 soundClassObject=/Game/Sounds/FortSoundClassesAndModes/Fort_Creatures_Steps_Normal
 volumeMultiplier=0.5
 concurrencySet [/Irwin/Sounds/Settings/Conc/SetOneLimitToOwner_Foley_Conc, /Irwin/Sounds/Settings/Conc/Prey_SetGroup_Foley_Conc]
}
