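# Sound cue for the Labrador creature's swim landing foley, reusing Athena third-person player swim water waves at a raised pitch.
# A quality-level switch gives two tiers a random pick of six waves and the other two a pick of two, with the smaller set excluded from branch culling.
# A modulator feeds the output, which uses Prey step attenuation, prey foley concurrency and the creature footsteps sound class.
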SoundCue /Labrador/Sound/Foley/Labrador_Foley_Swim_Land_Cue
root=Modulator_0
nodes [
 {id=Modulator_0, children [QualityLevel_0], properties {volumeMax=1, volumeMin=1}}
 {id=QualityLevel_0, children [Random_2,Random_2,Random_1,Random_1]}
 {id=Random_1, children [WavePlayer_12, WavePlayer_13], properties {bShouldExcludeFromBranchCulling=true, weights [1,1]}}
 {
  id=Random_2
  children [WavePlayer_6, WavePlayer_7, WavePlayer_8, WavePlayer_9, WavePlayer_10, WavePlayer_11]
  properties {weights [1,1,1,1,1,1]}
 }
 {id=WavePlayer_10, properties {soundWaveAssetPtr=/Game/Athena/Sounds/Water/Interaction/Water_Player_Swim_Normal_3P_05}}
 {id=WavePlayer_11, properties {soundWaveAssetPtr=/Game/Athena/Sounds/Water/Interaction/Water_Player_Swim_Normal_3P_06}}
 {id=WavePlayer_12, properties {soundWaveAssetPtr=/Game/Athena/Sounds/Water/Interaction/Water_Player_Swim_Normal_3P_01}}
 {id=WavePlayer_13, properties {soundWaveAssetPtr=/Game/Athena/Sounds/Water/Interaction/Water_Player_Swim_Normal_3P_02}}
 {id=WavePlayer_6, properties {soundWaveAssetPtr=&Water_Player_Swim_Normal_3P_01}}
 {id=WavePlayer_7, properties {soundWaveAssetPtr=&Water_Player_Swim_Normal_3P_02}}
 {id=WavePlayer_8, properties {soundWaveAssetPtr=/Game/Athena/Sounds/Water/Interaction/Water_Player_Swim_Normal_3P_03}}
 {id=WavePlayer_9, properties {soundWaveAssetPtr=/Game/Athena/Sounds/Water/Interaction/Water_Player_Swim_Normal_3P_04}}
]
properties {
 attenuationSettings=/Irwin/Sounds/Settings/Prey_Step_Att
 maxDistance=3200
 pitchMultiplier=1.5
 soundClassObject=/Game/Sounds/FortSoundClassesAndModes/Fort_Creatures_Steps_Normal
 volumeMultiplier=1
 concurrencySet [/Irwin/Sounds/Settings/Conc/Prey_SetGroup_Foley_Conc]
}
