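# Sound cue for the Labrador teleport-out sound, played as creature audio in the game.
# A single wave player feeds the teleport-out sound through a modulator that randomly raises its pitch on each play.
# The cue boosts volume, borrows the far prey gallop attenuation from Irwin, and routes to the tactical creature vocalization sound class despite sitting in the Foley folder.
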SoundCue /Labrador/Sound/Foley/Labrador_TeleportOut_Cue
root=Modulator_0
nodes [
 {id=Modulator_0, children [WavePlayer_0], properties {pitchMax=1.4, pitchMin=1.2}}
 {id=WavePlayer_0, properties {soundWaveAssetPtr=&/Labrador_TeleportOut}}
]
properties {
 attenuationSettings=/Irwin/Sounds/Settings/Prey_Gallop_Far_Att
 maxDistance=4462.9136
 soundClassObject=&Fort_Creatures_Vocalization_Tactical
 volumeMultiplier=2.4
}
