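# Stereo MetaSound for Labrador taming progression feedback, routed to the non-critical in-game UI sound class.
# On play it mixes a HUD progress wave, pitched by the PitchShift input, with a training-start wave.
# An ADSR envelope gates the mix, attacking on play or when PitchShift starts changing, then releasing and re-arming after changes pause briefly.
# The HUD wave drives OnFinished; CompletionTime and Sounds inputs are unconnected.
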MetaSoundSource /Labrador/Sound/Foley/MSS_Labrador_Taming_Progression
inputs [
 CompletionTime {type=Time}
 PitchShift {type=Float}
 Sounds {type=WaveAsset:Array}
 UE.Source.OnPlay {type=Trigger}
]
interfaces [UE.OutputFormat.Stereo:1.0, UE.Source.OneShot:1.0, UE.Source:1.1]
outputs [
 "UE.OutputFormat.Stereo.Audio:0" {type=Audio}
 "UE.OutputFormat.Stereo.Audio:1" {type=Audio}
 UE.Source.OneShot.OnFinished {type=Trigger}
]
pages [Default {
  input_defaults {CompletionTime=30, PitchShift=1, UE.Source.OnPlay=null, Sounds []}
  nodes [
   InitVariable.WaveAsset()
   TriggerOnValueChange_0 = UE.Trigger On Value Change:Float(Value=^PitchShift)
   TriggerDelay_0 = UE.Trigger Delay(In=@TriggerOnValueChange_0.Trigger) {also {defaults {Delay Time=0.5}}}
   TriggerOnce_0 = UE.Trigger Once(Reset=@TriggerDelay_0.Out, Trigger In=@TriggerOnValueChange_0.Trigger)
   TriggerAnyTriggerAny2_0 = "TriggerAny.Trigger Any (2)"(In 0=^UE.Source.OnPlay, In 1=@TriggerOnce_0.Trigger Out)
   ADSREnvelopeADSREnvelope_0 = ADSR Envelope.ADSR Envelope:Float(Trigger Attack=@TriggerAnyTriggerAny2_0.Out, Trigger Release=@TriggerDelay_0.Out) {also {defaults {Attack Time=0.3, Release Curve=1.2, Sustain Level=1}}}
   WavePlayer_0 = UE.Wave Player:Stereo(Play=^UE.Source.OnPlay, Pitch Shift=MapRange.MapRange:Float(In=^PitchShift) {also {defaults {Out Range A=-4, Out Range B=4}}}.Out Value) {also {defaults {Loop="<MetaSoundBuilderResult.SUCCEEDED: 0>", Wave Asset=&/SC_Labrador_Taming_HUD_01}, version [1,1]}}
   WavePlayer_1 = UE.Wave Player:Stereo(Play=^UE.Source.OnPlay) {also {defaults {Wave Asset=&/Labrador_RiftAnomaly_Training_Start_01}, version [1,1]}}
   AudioMixerAudioMixerStereo2_0 = "AudioMixer.Audio Mixer (Stereo, 2)"(In 0 L=@WavePlayer_0.Out Left, In 0 R=@WavePlayer_0.Out Right, In 1 L=@WavePlayer_1.Out Left, In 1 R=@WavePlayer_1.Out Right) {also {defaults {Gain 0=0.7, Gain 1=1.2}}}
   Multiply_0 = UE.Multiply:Audio by Float(AdditionalOperands=@ADSREnvelopeADSREnvelope_0.Out Envelope, PrimaryOperand=@AudioMixerAudioMixerStereo2_0.Out L)
   Multiply_1 = UE.Multiply:Audio by Float(AdditionalOperands=@ADSREnvelopeADSREnvelope_0.Out Envelope, PrimaryOperand=@AudioMixerAudioMixerStereo2_0.Out R)
  ]
  links [
   {dst="@UE.OutputFormat.Stereo.Audio:0", src=Multiply_0.Out}
   {dst="@UE.OutputFormat.Stereo.Audio:1", src=Multiply_1.Out}
   {dst="@UE.Source.OneShot.OnFinished", src=WavePlayer_0.On Finished}
  ]
 }]
properties {
 assetImportData=&self:AssetImportData
 numChannels=2
 outputFormat=Stereo
 sampleRate=48000
 soundClassObject=&Fort_UI_InGame_NonCritical
 virtualizationMode=SeekRestart
}
