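# Sound cue for the Labrador creature's death countdown, playing the teleport buildup sound wave.
# A single wave player feeds a modulator that slightly randomizes volume and holds pitch constant.
# It uses the tactical creature vocalization sound class and borrows the Prey gallop attenuation settings from the Irwin content.
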SoundCue /Labrador/Sound/Foley/SC_Labrador_DeathCountdown_Cue
root=Modulator_0
nodes [
 {id=Modulator_0, children [WavePlayer_0], properties {pitchMax=1, pitchMin=1, volumeMax=1.2, volumeMin=1}}
 {id=WavePlayer_0, properties {soundWaveAssetPtr=&/Labrador_BuildupToTeleport}}
]
properties {
 attenuationSettings=/Irwin/Sounds/Settings/Prey_Gallop_Att
 maxDistance=3200
 soundClassObject=&Fort_Creatures_Vocalization_Tactical
 volumeMultiplier=0.9
}
