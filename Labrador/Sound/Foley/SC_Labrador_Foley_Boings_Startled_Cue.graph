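# Labrador foley Sound Cue for startled gallop "boing" sounds, played as a creature-steps class sound.
# A Random node picks one of four equally weighted wave variations, then a Modulator varies pitch slightly downward before a Delay at default settings and a single-input Mixer.
# The cue raises overall pitch and volume, and uses shared Irwin prey gallop attenuation plus per-owner and prey-group foley concurrency limits.
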SoundCue /Labrador/Sound/Foley/SC_Labrador_Foley_Boings_Startled_Cue
root=Mixer_0
nodes [
 {id=Delay_1, children [Modulator_0]}
 {id=Mixer_0, children [Delay_1], properties {inputVolume [1]}}
 {id=Modulator_0, children [Random_0], properties {pitchMax=1, pitchMin=0.9}}
 {id=Random_0, children [WavePlayer_4,WavePlayer_5,WavePlayer_6,WavePlayer_7], properties {weights [1,1,1,1]}}
 {id=WavePlayer_4, properties {soundWaveAssetPtr=&/Labrador_Gallop_Boing_Startle_01}}
 {id=WavePlayer_5, properties {soundWaveAssetPtr=&/Labrador_Gallop_Boing_Startle_02}}
 {id=WavePlayer_6, properties {soundWaveAssetPtr=&/Labrador_Gallop_Boing_Startle_03}}
 {id=WavePlayer_7, properties {soundWaveAssetPtr=&/Labrador_Gallop_Boing_Startle_04}}
]
properties {
 attenuationSettings=/Irwin/Sounds/Settings/Prey_Gallop_Far_Att
 maxDistance=4462.9136
 pitchMultiplier=1.2
 soundClassObject=/Game/Sounds/FortSoundClassesAndModes/Fort_Creatures_Steps_Normal
 volumeMultiplier=1.3
 concurrencySet [/Irwin/Sounds/Settings/Conc/SetOneLimitToOwner_Foley_Conc, /Irwin/Sounds/Settings/Conc/Prey_SetGroup_Foley_Conc]
}
