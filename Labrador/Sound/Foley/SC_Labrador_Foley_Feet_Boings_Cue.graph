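# Sound cue for the Labrador creature's foot foley, playing springy gallop "boing" sounds.
# A Random node picks among six equally weighted boing waves, then feeds a Modulator for pitch variation, a Delay and a single-input Mixer.
# The cue raises overall pitch and volume, routes to the creature footsteps sound class, and borrows Irwin prey gallop attenuation and two foley concurrency limits.
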SoundCue /Labrador/Sound/Foley/SC_Labrador_Foley_Feet_Boings_Cue
root=Mixer_0
nodes [
 {id=Delay_1, children [Modulator_0]}
 {id=Mixer_0, children [Delay_1], properties {inputVolume [1]}}
 {id=Modulator_0, children [Random_0], properties {pitchMax=1, pitchMin=0.9}}
 {
  id=Random_0
  children [WavePlayer_0, WavePlayer_1, WavePlayer_2, WavePlayer_3, WavePlayer_16, WavePlayer_17]
  properties {weights [1,1,1,1,1,1]}
 }
 {id=WavePlayer_0, properties {soundWaveAssetPtr=&/Labrador_Gallop_Boing_06}}
 {id=WavePlayer_1, properties {soundWaveAssetPtr=&/Labrador_Gallop_Boing_05}}
 {id=WavePlayer_16, properties {soundWaveAssetPtr=&/Labrador_Gallop_Boing_02}}
 {id=WavePlayer_17, properties {soundWaveAssetPtr=&/Labrador_Gallop_Boing_01}}
 {id=WavePlayer_2, properties {soundWaveAssetPtr=&/Labrador_Gallop_Boing_04}}
 {id=WavePlayer_3, properties {soundWaveAssetPtr=&/Labrador_Gallop_Boing_03}}
]
properties {
 attenuationSettings=/Irwin/Sounds/Settings/Prey_Gallop_Far_Att
 maxDistance=4462.9136
 pitchMultiplier=1.4
 soundClassObject=/Game/Sounds/FortSoundClassesAndModes/Fort_Creatures_Steps_Normal
 volumeMultiplier=1.2
 concurrencySet [/Irwin/Sounds/Settings/Conc/SetOneLimitToOwner_Foley_Conc, /Irwin/Sounds/Settings/Conc/Prey_SetGroup_Foley_Conc]
}
