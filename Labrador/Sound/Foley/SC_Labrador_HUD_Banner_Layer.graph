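# Sound cue for a HUD banner layer in Labrador, built on a ping banner sound and assigned to the Fort_UI sound class.
# A mixer at the root takes a single wave player for SC_Labrador_Ping_Banner_01, and its second input is left empty.
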SoundCue /Labrador/Sound/Foley/SC_Labrador_HUD_Banner_Layer
root=Mixer_0
nodes [
 {id=Mixer_0, children [WavePlayer_0,null], properties {inputVolume [1,1]}}
 {id=WavePlayer_0, properties {soundWaveAssetPtr=&/SC_Labrador_Ping_Banner_01}}
]
properties {maxDistance=2097152, soundClassObject=&Fort_UI, volumeMultiplier=1}
