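# Sound cue for Labrador hit foley, playing rift anomaly impact sounds.
# A random node picks among six equally weighted impact waves and feeds a modulator that varies pitch and volume slightly.
# The cue is routed to the impacts sound class, uses the shared prop impact attenuation and has a boosted volume multiplier.
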SoundCue /Labrador/Sound/Foley/SC_Labrador_Hit_Cue
root=Modulator_2
nodes [
 {id=Modulator_2, children [Random_2], properties {volumeMin=1}}
 {id=Random_2, children [WavePlayer_6,WavePlayer_7,WavePlayer_8,WavePlayer_9,WavePlayer_0,WavePlayer_1], properties {weights [1,1,1,1,1,1]}}
 {id=WavePlayer_0, properties {soundWaveAssetPtr=&/Labrador_RiftAnomaly_Impact_05}}
 {id=WavePlayer_1, properties {soundWaveAssetPtr=&/Labrador_RiftAnomaly_Impact_06}}
 {id=WavePlayer_6, properties {soundWaveAssetPtr=&/Labrador_RiftAnomaly_Impact_01}}
 {id=WavePlayer_7, properties {soundWaveAssetPtr=&/Labrador_RiftAnomaly_Impact_02}}
 {id=WavePlayer_8, properties {soundWaveAssetPtr=&/Labrador_RiftAnomaly_Impact_03}}
 {id=WavePlayer_9, properties {soundWaveAssetPtr=&/Labrador_RiftAnomaly_Impact_04}}
]
properties {
 attenuationSettings=/Game/Sounds/Attenuation/Fort_Prop_Impact_Att
 maxDistance=3256
 soundClassObject=&Fort_Impacts
 volumeMultiplier=1.8
}
