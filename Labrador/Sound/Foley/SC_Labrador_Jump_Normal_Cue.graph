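# Sound cue for the Labrador creature's normal jump foley, playing one of three mid rift-anomaly jump waves.
# An equal-weight random node picks the wave, and a modulator varies its pitch and volume slightly upward on each play.
# It borrows the far prey-gallop attenuation from the Irwin content and routes to the tactical creature vocalization sound class with boosted volume.
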SoundCue /Labrador/Sound/Foley/SC_Labrador_Jump_Normal_Cue
root=Modulator_0
nodes [
 {id=Modulator_0, children [Random_0], properties {pitchMax=1.3, pitchMin=1, volumeMax=1.2, volumeMin=1}}
 {id=Random_0, children [WavePlayer_4,WavePlayer_5,WavePlayer_6], properties {weights [1,1,1]}}
 {id=WavePlayer_4, properties {soundWaveAssetPtr=&/Labrador_RiftAnomaly_Jump_Mid_01}}
 {id=WavePlayer_5, properties {soundWaveAssetPtr=&/Labrador_RiftAnomaly_Jump_Mid_02}}
 {id=WavePlayer_6, properties {soundWaveAssetPtr=&/Labrador_RiftAnomaly_Jump_Mid_03}}
]
properties {
 attenuationSettings=/Irwin/Sounds/Settings/Prey_Gallop_Far_Att
 maxDistance=4462.9136
 soundClassObject=&Fort_Creatures_Vocalization_Tactical
 volumeMultiplier=1.3
}
