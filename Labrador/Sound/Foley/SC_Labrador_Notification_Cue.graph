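# Sound cue for a Labrador notification, playing the Rift Anomaly UI banner sound in the Fort_UI sound class.
# A single wave player feeds a root Mixer, whose second input is empty, so the mixer adds no layering.
# No randomization, modulation or attenuation is set in the cue.
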SoundCue /Labrador/Sound/Foley/SC_Labrador_Notification_Cue
root=Mixer_0
nodes [
 {id=Mixer_0, children [WavePlayer_1,null], properties {inputVolume [1,1]}}
 {id=WavePlayer_1, properties {soundWaveAssetPtr=&/Labrador_RiftAnomaly_UI_Banner_01}}
]
properties {maxDistance=2097152, soundClassObject=&Fort_UI, volumeMultiplier=1}
