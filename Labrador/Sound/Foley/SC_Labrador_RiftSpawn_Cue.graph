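# Sound cue for a Labrador rift spawn foley effect, playing the rift anomaly spawn-in sound.
# A single wave player feeds a modulator that randomizes pitch upward from normal and varies volume slightly on each play.
# The cue uses the shared rift portal attenuation and the in-game SFX sound class, with a boosted overall volume.
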SoundCue /Labrador/Sound/Foley/SC_Labrador_RiftSpawn_Cue
root=Modulator_0
nodes [
 {id=Modulator_0, children [WavePlayer_0], properties {pitchMax=1.3, pitchMin=1, volumeMax=1.2, volumeMin=1}}
 {id=WavePlayer_0, properties {soundWaveAssetPtr=&/Labrador_RiftAnomaly_SpawnIn_01}}
]
properties {
 attenuationSettings=/Game/Sounds/Attenuation/Fort_RiftPortals_Att
 maxDistance=4000
 soundClassObject=&Fort_SFX_InGame
 volumeMultiplier=1.6
}
