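# Looping sound cue for the Labrador speed boost, playing a windy run-boost wave.
# The wave player feeds an Enveloper, whose volume references an external boost FX curve borrowed from the Antelope drivable vehicle, then a single-input Mixer.
# It uses the emotes attenuation and player-interact sound class, with a concurrency override allowing one instance at a time.
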SoundCue /Labrador/Sound/Foley/SC_Labrador_SpeedBoost_LP
root=Mixer_0
nodes [
 {
  id=Enveloper_0
  children [WavePlayer_1]
  properties {
   pitchCurve {editorCurveData {keys [{time=0}]}}
   volumeCurve {externalCurve=/Game/Athena/DrivableVehicles/Antelope/AntelopeBoostFX, editorCurveData {keys [<at 0 1>]}}
  }
 }
 {id=Mixer_0, children [Enveloper_0], properties {inputVolume [1]}}
 {id=WavePlayer_1, properties {bLooping=true, soundWaveAssetPtr=&/Labrador_Run_Boost_Windy_LP}}
]
properties {
 attenuationSettings=&Fort_Emotes_Athena_Att
 bOverrideConcurrency=true
 maxDistance=1600
 soundClassObject=/Game/Sounds/FortSoundClassesAndModes/Fort_08_Player_Interacts
 volumeMultiplier=1.2
 concurrencyOverrides {maxCount=1}
}
