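# Sound cue that plays a whip crack for the Labrador speed boost, filed under foley.
# A Random node picks one of two whip-crack waves with equal weight and allows repeats, then feeds a single-input Mixer at the root.
# It borrows Irwin's far prey-gallop attenuation, uses the tactical creature vocalization sound class, raises volume slightly and overrides concurrency to cap simultaneous plays.
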SoundCue /Labrador/Sound/Foley/SC_Labrador_SpeedBoost_Whip
root=Mixer_0
nodes [
 {id=Mixer_0, children [Random_0], properties {inputVolume [1]}}
 {id=Random_0, children [WavePlayer_0,WavePlayer_4], properties {bRandomizeWithoutReplacement=false, weights [1,1]}}
 {id=WavePlayer_0, properties {soundWaveAssetPtr=&/Labrador_Run_Boost_WhipCrack_02}}
 {id=WavePlayer_4, properties {soundWaveAssetPtr=&/Labrador_Run_Boost_WhipCrack_01}}
]
properties {
 attenuationSettings=/Irwin/Sounds/Settings/Prey_Gallop_Far_Att
 bOverrideConcurrency=true
 maxDistance=4462.9136
 soundClassObject=&Fort_Creatures_Vocalization_Tactical
 volumeMultiplier=1.3
 concurrencyOverrides {maxCount=4}
}
