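# Sound cue for the Labrador creature's untopple foley, built on a single RiftAnomaly "BackFromStunned" wave.
# The wave player feeds a Delay node left at default settings, then a Modulator that slightly randomizes pitch and volume upward on each play.
# The cue boosts overall volume, borrows Irwin's far prey-gallop attenuation and routes to the tactical creature vocalization sound class.
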SoundCue /Labrador/Sound/Foley/SC_Labrador_Untopple_Cue
root=Modulator_0
nodes [
 {id=Delay_0, children [WavePlayer_1]}
 {id=Modulator_0, children [Delay_0], properties {pitchMax=1.1, pitchMin=1, volumeMax=1.2, volumeMin=1}}
 {id=WavePlayer_1, properties {soundWaveAssetPtr=&/Labrador_RiftAnomaly_BackFromStunned_01}}
]
properties {
 attenuationSettings=/Irwin/Sounds/Settings/Prey_Gallop_Far_Att
 maxDistance=4462.9136
 soundClassObject=&Fort_Creatures_Vocalization_Tactical
 volumeMultiplier=1.5
}
